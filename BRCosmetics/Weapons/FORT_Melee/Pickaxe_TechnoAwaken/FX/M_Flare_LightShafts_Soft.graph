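# Translucent unlit Niagara sprite and mesh particle material for soft light-shaft flares in the TechnoAwaken pickaxe effects.
# Particle color RGB drives emissive through an adjustment function, while particle alpha multiplies the refined green channel of a light-shaft mask for opacity.
# A static switch selects random sub-UV cells per particle or plain UVs.
# Opacity gets depth fade and optional camera fade, both bypassed on mobile.
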Material /BRCosmetics/Weapons/FORT_Melee/Pickaxe_TechnoAwaken/FX/M_Flare_LightShafts_Soft
graph {
 properties {
  bAllowNegativeEmissiveColor=true
  bUsedWithNiagaraMeshParticles=true
  bUsedWithNiagaraSprites=true
  blendMode=BLEND_Translucent
  pixelDepthOffsetMode=PDOM_Legacy
  shadingModel=MSM_Unlit
  usedShadingModels=Unlit
 }
 nodes [
  n1 = ParticleColor()
  MaterialFunctionCall(Emissive=@n1.RGB) {materialFunction=&MF_FX_EmissiveAdjustments}
  n2 = StaticSwitchParameter(
   True=MaterialFunctionCall!(
    UseTexCoord 1
    Rows=ScalarParameter() {defaultValue=1, parameterName=Rows}
    Columns=ScalarParameter() {defaultValue=4, parameterName=Columns}
    Particle Random Value=ParticleRandom()
   ) {materialFunction=/Game/Effects/Fort_Effects/Materials/_Functions/MF_RandomSubUV_GPU_Sprites}.Result
   False=TextureCoordinate()
  ) {parameterName=Use GPU Random}
  n3 = VectorParameter() {parameterName=Refine, defaultValue=<rgba 1 1 1 0.5>}
  n4 = Multiply(
   A=@n1.A
   B=MaterialFunctionCall(Texture=TextureSample(UVs=@n2) {texture=/Game/Effects/Fort_Effects/Textures/Mask/T_Flare_LightShafts}.G, RGB=@n3.RGB, A=@n3.RGB) {materialFunction=&MF_FX_Refine}.Result
  )
  n5 = DepthFade(Opacity=@n4, FadeDistance=ScalarParameter() {defaultValue=100, parameterName=DepthFade})
  n6 = MaterialFunctionCall(
   Fade Length=ScalarParameter() {defaultValue=1000, parameterName=CameraFadeLength}
   Fade Offset=ScalarParameter() {parameterName=CameraFadeOffset}
  ) {materialFunction=&CameraDepthFade}
  ShadingPathSwitch(Default=StaticSwitchParameter(True=Multiply(A=@n5, B=@n6.Result), False=@n5) {parameterName=UseCameraFade}, Mobile=@n4) {desc=Migrated from FeatureLevelSwitch}
 ]
}
