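# Translucent unlit effect material for the TechnoAwaken pickaxe, usable on Niagara mesh particles and skeletal meshes to draw a glowing, hue-varying version of the weapon.
# The pickaxe diffuse texture is saturation-adjusted, powered and tinted, then hue-shifted by Perlin noise, scaled by particle color and passed through emissive adjustments.
# A separate alpha chain multiplies particle alpha by a normal-mapped Fresnel rim and a local-Z gradient.
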Material /BRCosmetics/Weapons/FORT_Melee/Pickaxe_TechnoAwaken/FX/M_PxSimpleMeshMat
graph {
 properties {
  bUsedWithMeshDeformer=true
  bUsedWithNiagaraMeshParticles=true
  bUsedWithSkeletalMesh=true
  blendMode=BLEND_Translucent
  pixelDepthOffsetMode=PDOM_Legacy
  shadingModel=MSM_Unlit
  usedShadingModels=Unlit
 }
 nodes [
  n1 = TextureObjectParameter() {
   group=H
   parameterName=H_Texture
   samplerType=SAMPLERTYPE_LinearGrayscale
   sortPriority=0
   texture=&T_SimplePerlin_Greyscale
  }
  n2 = VectorParameter() {group=H, parameterName=H_UVControl, channelNames {a=VPan, b=UPan, g=VScale, r=UScale}, defaultValue=<rgba 0.5 0.5 0.4 0.5>}
  n3 = MaterialFunctionCall(Texture=@n1, UVControl_RGB=@n2, UVControl_A=@n2) {materialFunction=&FXCommon_TextureSample}
  n4 = Desaturation(
   None=TextureSample() {texture=/BRCosmetics/Weapons/FORT_Melee/Pickaxe_TechnoAwaken/Textures/T_Pickaxe_TechnoAwaken_D}.RGB
   Fraction=ScalarParameter() {defaultValue=-2, parameterName=Desa}
  )
  n5 = ScalarParameter() {defaultValue=0.2, parameterName=sssssP}
  n6 = MaterialFunctionCall(
   Hue Shift Percentage=Multiply(A=Subtract(A=@n3.R) {constB=0.5}, B=ScalarParameter() {defaultValue=0.3, parameterName=H})
   Texture=Multiply(A=Power(Base=@n4, Exp=@n5), B=VectorParameter() {defaultValue=<rgba 2.5 1 10 1>}.RGB)
  ) {materialFunction=&HueShift}
  n7 = ParticleColor()
  n8 = VectorParameter() {parameterName=TW, defaultValue=<rgba 0.18 3.3 0 1>}
  MaterialFunctionCall(Emissive=Multiply(A=Multiply(A=@n6.Result, B=@n7.RGB), B=@n8.R)) {materialFunction=&MF_FX_EmissiveAdjustments}
  n9 = VectorParameter() {parameterName=Test, defaultValue=<rgba 7 0.04 0 1>}
  n10 = Fresnel(
   ExponentIn=@n9.R
   BaseReflectFractionIn=@n9.R
   Normal=Transform(None=TextureSample() {samplerType=SAMPLERTYPE_Normal, texture=/BRCosmetics/Weapons/FORT_Melee/Pickaxe_TechnoAwaken/Textures/T_Pickaxe_TechnoAwaken_N}.RGB)
  ) {baseReflectFraction=0.02, exponent=7}
  n11 = MaterialFunctionCall(Float3=MaterialFunctionCall() {materialFunction=&LocalPosition}."Instance Local Position (Excluding Offsets)") {materialFunction=&BreakOutFloat3Components}
  n12 = ScalarParameter() {defaultValue=100, parameterName=stest}
  n13 = Power(Base=Saturate(None=Divide(A=@n11.B, B=@n12))) {constExponent=1.5}
  Multiply(A=Multiply(A=@n7.A, B=Multiply(A=@n10, B=@n13)), B=@n8.G)
 ]
}
