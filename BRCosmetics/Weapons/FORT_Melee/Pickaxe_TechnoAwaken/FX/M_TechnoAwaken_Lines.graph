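# Translucent, unlit, two-sided material for Niagara ribbons in the TechnoAwaken pickaxe effects, drawing procedural repeating lines.
# A normalized sine along the V coordinate is raised to two exponents, and the scaled results are blended into line intensity.
# Particle alpha, edge masks on both UV axes and a depth fade shape opacity.
# Particle color times a Color parameter feeds emissive through a shared adjustment function.
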Material /BRCosmetics/Weapons/FORT_Melee/Pickaxe_TechnoAwaken/FX/M_TechnoAwaken_Lines
graph {
 properties {
  bUsedWithNiagaraRibbons=true
  blendMode=BLEND_Translucent
  pixelDepthOffsetMode=PDOM_Legacy
  shadingModel=MSM_Unlit
  twoSided=true
  usedShadingModels=Unlit
 }
 nodes [
  n1 = MaterialFunctionCall(Float2=TextureCoordinate()) {materialFunction=&BreakOutFloat2Components}
  n2 = VectorParameter() {group=edge2, parameterName=edge2_EdgeMask_InvertControl, channelNames {a="", b=Power, g=Multiply, r=Offset}, defaultValue=<rgba 0 4 2 1>}
  n3 = VectorParameter() {group=edge2, parameterName=edge2_EdgeMask_Control, channelNames {a="", b=Power, g=Multiply, r=Offset}, defaultValue=<rgba 0 15 3 1>}
  n4 = VectorParameter() {group=Edge, parameterName=Edge_EdgeMask_Control, channelNames {a="", b=Power, g=Multiply, r=Offset}, defaultValue=<rgba 0 2 1 1>}
  n5 = MaterialFunctionCall!(UorV=@n1.G, EdgeControl=@n4) {materialFunction=/Game/Effects/Niagara/MasterMaterials/Functions/Common/FXCommon_EdgeMask_Simple}
  n6 = ParticleColor()
  n7 = Multiply(
   A=MaterialFunctionCall(Float2=TextureCoordinate()) {materialFunction=&BreakOutFloat2Components}.G
   B=ScalarParameter() {defaultValue=5.5, parameterName=F}
  ) {constB=5}
  n8 = ScalarParameter() {defaultValue=1, parameterName=W}
  n9 = Divide(A=Add(A=Sine(None=@n7), B=@n8), B=Add(A=@n8))
  n10 = VectorParameter() {parameterName=R, defaultValue=<rgba 150 0.5 0.8 0.08>}
  n11 = MaterialFunctionCall(Float3=@n10.RGB) {materialFunction=&BreakOutFloat3Components}
  n12 = Power(Base=@n9, Exp=@n11.R)
  n13 = Power(Base=@n9, Exp=@n11.B)
  n14 = Multiply(
   A=@n6.A
   B=LinearInterpolate(A=Multiply(A=@n12, B=@n11.G), B=Multiply(A=@n13, B=@n10.A), Alpha=ScalarParameter() {defaultValue=0.5, parameterName=A})
  )
  DepthFade(Opacity=Multiply(
    A=MaterialFunctionCall!(UorV=@n1.R, Invert_EdgeControl=@n2, NoInvert_EdgeControl=@n3) {materialFunction=/Game/Effects/Niagara/MasterMaterials/Functions/Common/FXCommon_EdgeMask_Ex}.Mask
    B=Multiply(A=@n5.Mask, B=@n14) {constA=1}
   ) {constA=1}) {fadeDistanceDefault=5}
  MaterialFunctionCall(Emissive=Multiply(A=VectorParameter() {parameterName=Color, defaultValue=<rgba 1 1 1 1>}.RGB, B=@n6.RGB)) {materialFunction=&MF_FX_EmissiveAdjustments}
 ]
}
